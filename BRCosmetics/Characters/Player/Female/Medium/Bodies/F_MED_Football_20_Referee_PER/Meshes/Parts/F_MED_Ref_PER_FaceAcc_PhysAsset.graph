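# Physics asset for the face accessory mesh of the female medium football referee cosmetic, giving its dynamic hair bones capsule bodies and joint limits.
# Larger capsules on head, spine_05 and left clavicle cover the upper body, linked by free angular constraints.
# Three smaller hair bodies constrain to the head with swing and twist limits.
# The main hair root constrains to spine_05 rather than the head, and its two child bodies branch from it with tight twist.
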PhysicsAsset /BRCosmetics/Characters/Player/Female/Medium/Bodies/F_MED_Football_20_Referee_PER/Meshes/Parts/F_MED_Ref_PER_FaceAcc_PhysAsset
bodies {
 clavicle_l [{shapeType=Capsule, length=11.0514, radius=6.5365, center=<7.533 -0.3482 -2.2909>, rotation=<rot 85.0779 0 0>}]
 dyn_hair_br [{shapeType=Capsule, length=4.132, radius=2, center=<0.3685 0 0>, rotation=<rot 90 0 0>}]
 dyn_hair_lb [{shapeType=Capsule, length=4.4229, radius=2, center=<0.6935 0.0001 0>, rotation=<rot 90 0 0>}]
 dyn_hair_lt [{shapeType=Capsule, length=1.7789, radius=2, center=<0.3239 0 0>, rotation=<rot 90 0 0>}]
 dyn_hair_main_01 [{shapeType=Capsule, length=4.079, radius=4.3481, center=<3 1 0>, rotation=<rot 84.9999 180 180>}]
 dyn_hair_main_02 [{shapeType=Capsule, length=2.25, radius=3.1678, center=<1 0 0>, rotation=<rot 90 0 0>}]
 dyn_hair_main_03 [{shapeType=Capsule, length=3.2988, radius=3.0077, center=<3 0 0>, rotation=<rot 84.9999 180 180>}]
 head [{shapeType=Capsule, length=4.1186, radius=8.3511, center=<4.6565 -0.6436 -2.4792>, rotation=<rot 90 0 0>}]
 spine_05 [{shapeType=Capsule, length=8.5477, radius=18.177, center=<-5.161 -8.7666 0>}]
}
constraints [
 {bone1Name=dyn_hair_main_02, bone2Name=dyn_hair_main_01, swing1LimitDegrees=45, swing2LimitDegrees=45, twistLimitDegrees=2}
 {bone1Name=dyn_hair_main_03, bone2Name=dyn_hair_main_01, swing1LimitDegrees=45, swing2LimitDegrees=45, twistLimitDegrees=2}
 {bone1Name=dyn_hair_br, bone2Name=head, swing1LimitDegrees=45, swing2LimitDegrees=45, twistLimitDegrees=15}
 {bone1Name=dyn_hair_lb, bone2Name=head, swing1LimitDegrees=45, swing2LimitDegrees=45, twistLimitDegrees=15}
 {bone1Name=dyn_hair_lt, bone2Name=head, swing1LimitDegrees=45, swing2LimitDegrees=45, twistLimitDegrees=15}
 {
  bone1Name=head
  bone2Name=spine_05
  swing1LimitDegrees=45
  swing1Motion=Free
  swing2LimitDegrees=45
  swing2Motion=Free
  twistLimitDegrees=45
  twistMotion=Free
 }
 {
  bone1Name=clavicle_l
  bone2Name=spine_05
  swing1LimitDegrees=45
  swing1Motion=Free
  swing2LimitDegrees=45
  swing2Motion=Free
  twistLimitDegrees=45
  twistMotion=Free
 }
 {bone1Name=dyn_hair_main_01, bone2Name=spine_05, swing1LimitDegrees=30, swing2LimitDegrees=30, twistLimitDegrees=5}
]
